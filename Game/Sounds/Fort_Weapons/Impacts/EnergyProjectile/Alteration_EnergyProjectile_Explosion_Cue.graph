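# Sound cue for the explosion impact of the Alteration energy projectile weapon effect.
# A random node picks one of five equally weighted explosion wave variations per play.
# Playback is boosted in volume, routed through the Fort impacts sound class and spatialized with ability-radius attenuation.
# It references the rifle bullet impact concurrency set and carries override settings that cap simultaneous instances and block new ones.
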SoundCue /Game/Sounds/Fort_Weapons/Impacts/EnergyProjectile/Alteration_EnergyProjectile_Explosion_Cue
root=Random_1
nodes [
 {id=Random_1, children [WavePlayer_6, WavePlayer_7, WavePlayer_8, WavePlayer_9, WavePlayer_10], properties {weights [1,1,1,1,1]}}
 {id=WavePlayer_10, properties {soundWaveAssetPtr=&/Alteration_EnergyProjectile_Explosion_05}}
 {id=WavePlayer_6, properties {soundWaveAssetPtr=&/Alteration_EnergyProjectile_Explosion_02}}
 {id=WavePlayer_7, properties {soundWaveAssetPtr=&/Alteration_EnergyProjectile_Explosion_01}}
 {id=WavePlayer_8, properties {soundWaveAssetPtr=&/Alteration_EnergyProjectile_Explosion_03}}
 {id=WavePlayer_9, properties {soundWaveAssetPtr=&/Alteration_EnergyProjectile_Explosion_04}}
]
properties {
 attenuationSettings=/Game/Sounds/Attenuation/Fort_Ability_Radius
 maxDistance=1200
 soundClassObject=&Fort_Impacts
 volumeMultiplier=2.854
 concurrencyOverrides {maxCount=4, resolutionRule=PreventNew}
 concurrencySet [&RifleBulletImpact_Conc]
}
